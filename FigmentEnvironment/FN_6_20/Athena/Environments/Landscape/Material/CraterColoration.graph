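# Material function that tints crater landscape base color with a gradient texture driven by a painted layer weight.
# Sine of base-color red times pi, scaled by GradientModulation, offsets LayerWeight to form the gradient lookup UV, varying the tint with surface detail.
# The gradient multiplies base color, output as material attributes, plain color and raw gradient.
# Scorched-earth, gravel and Crater layer-weight nodes remain unconnected to any output.
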MaterialFunction /FigmentEnvironment/FN_6_20/Athena/Environments/Landscape/Material/CraterColoration
graph {nodes [
  n1 = FunctionInput() {inputType=FunctionInput_MaterialAttributes}
  n2 = MaterialFunctionCall!(Material=@n1) {materialFunction=/Engine/Functions/MaterialLayerFunctions/MatLayerBlend_BreakBaseColor}
  n3 = Sine(None=MaterialFunctionCall(Multiplier=ComponentMask(None=@n2.BaseColor) {r=true}) {materialFunction=&Pi}.Result)
  n4 = ScalarParameter() {defaultValue=0.0476, parameterName=GradientModulation}
  n5 = FunctionInput() {
   bUsePreviewValueAsDefault=true
   desc=Gradient INPUT from Landscape Paint
   inputName=LayerWeight
   inputType=FunctionInput_Scalar
   sortPriority=4
  }
  n6 = TextureSample(UVs=Add(A=Multiply(A=@n3, B=@n4), B=@n5)) {texture=/FigmentEnvironment/FN_6_20/Packages/DS_Fortnight/T_Crater_Gradient2}
  FunctionOutput(None=MaterialFunctionCall!(Material=@n1, NewBaseColor=Multiply(A=@n2.BaseColor, B=@n6.RGB), Mask) {materialFunction=/Engine/Functions/MaterialLayerFunctions/MatLayerBlend_OverrideBaseColor}.Blended Material) {outputName=MaterialOutput}
  n7 = FunctionInput() {inputName=BaseColor, sortPriority=1}
  n8 = Sine(None=MaterialFunctionCall(Multiplier=ComponentMask(None=@n7) {r=true}) {materialFunction=&Pi}.Result)
  n9 = TextureSample(UVs=Add(A=Multiply(A=@n8, B=@n4), B=@n5)) {texture=&T_Crater_Gradient2}
  FunctionOutput(None=Multiply(A=@n7, B=@n9.RGB)) {outputName=BaseColorOutput, sortPriority=1}
  FunctionOutput(None=@n9.RGB) {outputName=Gradient, sortPriority=5}
  LandscapeLayerWeight!(Base=Constant(), Layer=Constant() {r=1}) {parameterName=Crater}
  LandscapeLayerWeight(Base=Constant(), Layer=Constant() {r=1}) {parameterName=Crater}
  n10 = ScalarParameter() {defaultValue=-0.2522, parameterName=MudCraterTransition_High}
  n11 = ScalarParameter() {defaultValue=50, parameterName=MudCraterTransition_LOW}
  n12 = LandscapeLayerWeight(Base=Constant(), Layer=Constant() {r=1}) {parameterName=Crater}
  LinearInterpolate(Alpha=Clamp(None=LinearInterpolate(A=@n10, B=@n11, Alpha=@n12)))
  n13 = Divide(A=LandscapeLayerCoords(), B=ScalarParameter() {defaultValue=6.1467, parameterName=Tiling})
  n14 = TextureSample(UVs=@n13) {texture=/FigmentEnvironment/FN_6_20/Packages/DS_Fortnight/T_Crater_Scorched_Earth_D}
  n15 = ScalarParameter() {defaultValue=0.1, parameterName=S}
  n16 = ScalarParameter() {defaultValue=0.2, parameterName=S_1}
  n17 = TextureSampleParameter2D(UVs=Divide(A=LandscapeLayerCoords() {mappingScale=4}, B=ScalarParameter() {defaultValue=1.7165, parameterName=Tiling_1})) {
   group=Gravel
   parameterName=Gravel_1
   samplerSource=SSM_Wrap_WorldGroupSettings
   texture=/FigmentEnvironment/FN_6_20/Packages/DS_Fortnight/WhiteRockCustom_D
  }
  MaterialFunctionCall!(
   Base Material=MakeMaterialAttributes!(
    BaseColor=@n14.RGB
    Metallic
    Specular=LinearInterpolate(A=@n15, B=@n16, Alpha=@n14.R)
    Roughness
    Anisotropy
    EmissiveColor
    Opacity
    OpacityMask
    Normal=TextureSample(UVs=@n13) {samplerType=SAMPLERTYPE_Normal, texture=/FigmentEnvironment/FN_6_20/Packages/DS_Fortnight/T_Crater_Scorched_Earth_N}.RGB
    Tangent
    WorldPositionOffset
    SubsurfaceColor
    ClearCoat
    ClearCoatRoughness
    AmbientOcclusion
    Refraction
    CustomizedUVs_0
    CustomizedUVs_1
    CustomizedUVs_2
    CustomizedUVs_3
    CustomizedUVs_4
    CustomizedUVs_5
    CustomizedUVs_6
    CustomizedUVs_7
    PixelDepthOffset
    ShadingModel
    Displacement
   )
   Top Material=MakeMaterialAttributes(BaseColor=@n17.RGB, Specular=Constant() {r=0.2})
   Alpha=LandscapeLayerWeight(Base=Constant(), Layer=Constant() {r=1}) {parameterName=Gravel}
  ) {materialFunction=/Engine/Functions/MaterialLayerFunctions/MatLayerBlend_Standard}
  TextureSample(UVs=Divide(A=LandscapeLayerCoords(), B=ScalarParameter() {defaultValue=6.1467, parameterName=Tiling_2})) {texture=&T_Crater_Scorched_Earth_D}
 ]}
